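# SoundCue for the UI sound played when the team score XP bar appears on the reward screen.
# A single wave player plays the matching bar-appear sound wave directly, with no randomization or modulation.
# The cue is routed through the Fort_UI sound class and has a very large max distance.
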SoundCue /FNGameplayCues/Sounds/Fort_UI_Sounds/Rewards/Fort_RewardScreen_TeamScore_XP_Bar_Appear_Cue
root=WavePlayer_0
nodes [{id=WavePlayer_0, properties {soundWaveAssetPtr=&/Fort_RewardScreen_TeamScore_XP_Bar_Appear}}]
properties {maxDistance=2097152, soundClassObject=&Fort_UI}
